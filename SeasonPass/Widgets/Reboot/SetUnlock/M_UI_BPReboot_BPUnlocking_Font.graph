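# Translucent UI material for font in the Battle Pass Reboot unlocking widget, judging by its name and path, building a three-color gradient.
# Two rotated UV gradients, shaped by SmoothStep, Power and a shared WhiteGradient_Opacity parameter, drive nested lerps between Color 1, 2 and 3.
# A time-panned, rotated screen-space noise sample feeds a Distortion reroute that nothing reads.
# No connection from the color blend to a material output is shown.
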Material /SeasonPass/Widgets/Reboot/SetUnlock/M_UI_BPReboot_BPUnlocking_Font
graph {
 properties {
  bAutomaticallySetUsageInEditor=false
  bUsedWithStaticMesh=false
  blendMode=BLEND_Translucent
  materialDomain=MD_UI
  usedShadingModels=Default Lit
 }
 nodes [
  n1 = Constant() {r=0.5}
  n2 = Subtract(B=@n1) {constA=0.5}
  n3 = Add(B=@n1) {constA=0.5}
  n4 = ComponentMask(None=MaterialFunctionCall(UVs=TextureCoordinate(), "Rotation Angle (0-1)"=Constant() {r=0.7}) {materialFunction=&CustomRotator}.Rotated Values) {g=true}
  n5 = SmoothStep(Value=Multiply(A=SmoothStep(Min=@n2, Max=@n3, Value=@n4), B=ScalarParameter() {defaultValue=1, parameterName=WhiteGradient_Opacity}) {constB=0.5}) {constMin=0.25}
  n6 = Constant() {r=0.5}
  n7 = MaterialFunctionCall(UVs=TextureCoordinate(), "Rotation Angle (0-1)"=Constant() {r=0.8}) {materialFunction=&CustomRotator}
  n8 = Multiply(
   A=SmoothStep(Min=Subtract(B=@n6) {constA=0.5}, Max=Add(B=@n6) {constA=0.5}, Value=ComponentMask(None=@n7.Rotated Values) {g=true})
   B=ScalarParameter() {defaultValue=1, parameterName=WhiteGradient_Opacity}
  ) {constB=0.5}
  LinearInterpolate(
   A=VectorParameter() {parameterName=Color 1, defaultValue=<rgba 0.9473 0.2423 1 1>}.RGB
   B=LinearInterpolate(
    A=VectorParameter() {parameterName=Color 2, defaultValue=<rgba 0.2582 0.5711 1 1>}.RGB
    B=VectorParameter() {parameterName=Color 3, defaultValue=<rgba 0.6834 0.8258 1 1>}.RGB
    Alpha=Power(Base=@n5) {constExponent=0.5}
   )
   Alpha=Power(Base=SmoothStep(Value=@n8) {constMax=0.5}) {constExponent=0.5}
  )
  NamedRerouteDeclaration(None=Multiply(A=Time(), B=Constant() {r=0.4})) {name=Time, nodeColor=<rgba 0 0 0 1>}
  n9 = Panner(
   Coordinate=MaterialFunctionCall(UVs=ScreenPosition().ViewportUV, "Rotation Angle (0-1)"=Constant() {r=0.125}) {materialFunction=&CustomRotator}.Rotated Values
   Time=NamedRerouteUsage()
  ) {speedX=-0.2, speedY=0.1}
  n10 = Constant2Vector() {g=3, r=1}
  NamedRerouteDeclaration(None=TextureSample(UVs=MaterialFunctionCall(UVs=@n9, Texture Scale=@n10) {materialFunction=&ScaleUVsByCenter}.UVs) {samplerType=SAMPLERTYPE_Grayscale, texture=&LowResBlurredNoise}.R) {name=Distortion, nodeColor=<rgba 0 0 0 1>}
 ]
}
