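# Translucent UI font material for the Season Pass Reboot unlocking screen, producing a tinted, hue-shifted outline around glyphs.
# Font signed distance feeds MF_UI_ShapeOutline for outline alpha, scaled by an opacity parameter and a rotated screen-space gradient.
# Color 1 blends with a rotated UV gradient, then passes through HueShift and a negative-default Desaturation that boosts saturation.
# Named reroutes supply time and a panning screen-space blurred-noise distortion.
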Material /SeasonPass/Widgets/Reboot/SetUnlock/M_UI_BPReboot_BPUnlocking_Font_Outline
graph {
 properties {
  bAutomaticallySetUsageInEditor=false
  bUsedWithStaticMesh=false
  blendMode=BLEND_Translucent
  materialDomain=MD_UI
  usedShadingModels=Default Lit
 }
 nodes [
  n1 = Multiply(A=NamedRerouteUsage()) {constB=2}
  n2 = Multiply(A=NamedRerouteUsage()) {constB=0.5}
  n3 = VectorParameter() {parameterName=Color 1, defaultValue=<rgba 0.4125 0.3515 0.8469 1>}
  n4 = Constant() {r=0.5}
  n5 = MaterialFunctionCall(UVs=TextureCoordinate(), "Rotation Angle (0-1)"=Constant() {r=0.4}) {materialFunction=&CustomRotator}
  n6 = Multiply(
   A=SmoothStep(Min=Subtract(B=@n4) {constA=0.5}, Max=Add(B=@n4) {constA=0.5}, Value=ComponentMask(None=@n5.Rotated Values) {g=true})
   B=ScalarParameter() {defaultValue=0.4, parameterName=WhiteGradient_Opacity}
  ) {constB=0.5}
  Desaturation(
   None=MaterialFunctionCall(Hue Shift Percentage=Add(A=@n1, B=@n2), Texture=LinearInterpolate(A=@n3.RGB, Alpha=@n6)) {materialFunction=&HueShift}.Result
   Fraction=ScalarParameter() {defaultValue=-1, parameterName=RainbowSaturation}
  )
  n7 = FontSignedDistance()
  n8 = ScalarParameter() {defaultValue=0.01, parameterName=Line Tickness}
  n9 = NamedRerouteDeclaration(None=MaterialFunctionCall!(
    DistanceField=Multiply(A=@n7.Signed Distance, B=@n7.Signed Distance)
    Threshold
    Thickness=Multiply(A=@n8, B=@n7.Pixel Distance Factor)
    Alignment=Constant() {r=1}
    Smoothing
   ) {materialFunction=/FortUILibrary/MaterialFunctions/Shapes/Modifiers/MF_UI_ShapeOutline}.Outline) {name=Outline Alpha, nodeColor=<rgba 1 0.6563 0 1>}
  n10 = NamedRerouteUsage()
  n11 = MaterialFunctionCall(UVs=ScreenPosition().ViewportUV, "Rotation Angle (0-1)"=Constant() {r=0.4}) {materialFunction=&CustomRotator}
  Multiply(
   A=Multiply(A=Multiply(A=@n9, B=@n10), B=ScalarParameter() {defaultValue=0.5, parameterName=Outline Opacity})
   B=Power(Base=ComponentMask(None=@n11.Rotated Values) {g=true})
  )
  NamedRerouteDeclaration(None=Multiply(A=Time(), B=Constant() {r=0.4})) {name=Time, nodeColor=<rgba 0 0 0 1>}
  n12 = Panner(
   Coordinate=MaterialFunctionCall(UVs=ScreenPosition().ViewportUV, "Rotation Angle (0-1)"=Constant() {r=0.125}) {materialFunction=&CustomRotator}.Rotated Values
   Time=NamedRerouteUsage()
  ) {speedX=-0.2, speedY=0.1}
  n13 = Constant2Vector() {g=3, r=1}
  NamedRerouteDeclaration(None=TextureSample(UVs=MaterialFunctionCall(UVs=@n12, Texture Scale=@n13) {materialFunction=&ScaleUVsByCenter}.UVs) {samplerType=SAMPLERTYPE_Grayscale, texture=&LowResBlurredNoise}.R) {name=Distortion, nodeColor=<rgba 0 0 0 1>}
 ]
}
