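# UI translucent material for the character image on the Battle Pass Reboot set-unlocking widget, with a soft, slightly tilted fade along its bottom.
# The SetImage texture parameter is sampled with UI UVs scaled to fit, defaulting to a soldier commando image.
# The same UVs are slightly rotated, and their inverted vertical component passes through a SmoothStep to form a gradient mask that multiplies the image's alpha.
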Material /SeasonPass/Widgets/Reboot/SetUnlock/M_UI_BPReboot_SetUnlocking_Character
graph {
 properties {
  bAutomaticallySetUsageInEditor=false
  bUsedWithStaticMesh=false
  blendMode=BLEND_Translucent
  materialDomain=MD_UI
  usedShadingModels=Default Lit
 }
 nodes [
  n1 = MaterialFunctionCall(UV=MaterialFunctionCall() {materialFunction=&GetUserInterfaceUV}.Normalized UV) {materialFunction=&MF_UI_UVScaling}
  n2 = ComponentMask(None=MaterialFunctionCall(UVs=@n1.Scaled to Fit, "Rotation Angle (0-1)"=Constant() {r=0.03}) {materialFunction=&CustomRotator}.Rotated Values) {g=true}
  Multiply(
   A=TextureSampleParameter2D(UVs=@n1.Scaled to Fit) {parameterName=SetImage, texture=/FortUILibrary/Systems/Items/Textures/T_UI_SoldierCommando_F}.A
   B=SmoothStep(Value=OneMinus(None=@n2)) {constMax=0.4, constMin=0.1}
  )
 ]
}
