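# Translucent UI material for the Season Pass Reboot set-unlock modal, drawing a revealable horizontal band with glowing edge lines over a radial screen overlay.
# Modal Reveal Amount and Modal Size scale the UVs behind the band mask, while Line Reveal animates the lines.
# Named reroutes carry a dark base gradient, panning two-color rift underglow, corner blobs and line colors into lerp and color-dodge composites, finished with a slight random hue shift.
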Material /SeasonPass/Widgets/Reboot/SetUnlock/M_UI_BPReboot_SetUnlocking_Modal
graph {
 properties {
  bAutomaticallySetUsageInEditor=false
  bUsedWithStaticMesh=false
  blendMode=BLEND_Translucent
  materialDomain=MD_UI
  usedShadingModels=Default Lit
 }
 nodes [
  n1 = MaterialFunctionCall() {materialFunction=&GetUserInterfaceUV}
  n2 = Constant() {r=0.5}
  n3 = LinearInterpolate(A=Constant() {r=0.5}, Alpha=Power(Base=Distance(A=@n1.Normalized UV, B=@n2)))
  n4 = ScalarParameter() {defaultValue=1, parameterName=OverlayOpacity}
  LinearInterpolate(A=NamedRerouteUsage(), Alpha=LinearInterpolate(A=Multiply(A=@n3, B=@n4), Alpha=NamedRerouteUsage()))
  n5 = LinearInterpolate(A=NamedRerouteUsage(), B=NamedRerouteUsage(), Alpha=NamedRerouteUsage()) {constB=0}
  n6 = NamedRerouteUsage()
  n7 = NamedRerouteUsage()
  MaterialFunctionCall(
   Hue Shift Percentage=Multiply(A=MaterialFunctionCall!(UV, Seed) {materialFunction=&MF_UI_Random}.Result) {constB=0.035}
   Texture=LinearInterpolate(A=LinearInterpolate(A=@n5, B=@n6, Alpha=@n7), B=NamedRerouteUsage(), Alpha=NamedRerouteUsage())
  ) {materialFunction=&HueShift}
  MaterialFunctionCall() {materialFunction=&DebugTimeSine}
  n8 = Constant3Vector() {constant=<rgba 0.0048 0.004 0.0116 1>}
  n9 = MaterialFunctionCall(GradientStart=Constant() {r=-0.2}, GradientEnd=Constant() {r=1.2}, RotationAngle=Constant() {r=0.3}) {materialFunction=&MF_UI_LinearGradient}
  NamedRerouteDeclaration(None=LinearInterpolate(
    A=Power(Base=@n8) {constExponent=1.25}
    B=@n8
    Alpha=MaterialFunctionCall("Value (0-1)"=@n9.Result) {materialFunction=&MF_UI_EaseCurves}.EaseInOutSine
   )) {name=Base Gradient, nodeColor=<rgba 0 0 0 1>}
  n10 = VectorParameter() {group=0 - Coloring, parameterName=Color 1, defaultValue=<rgba 0 0.4025 1 1>}
  n11 = MaterialFunctionCall(UVs=NamedRerouteUsage(), Texture Scale=Constant() {r=2}) {materialFunction=&ScaleUVsByCenter}
  n12 = Frac(None=MaterialFunctionCall(UV Channel=Panner(Coordinate=@n11.UVs) {speedX=-0.2}) {materialFunction=&LinearGradient}.UGradient)
  n13 = Constant() {r=0.5}
  n14 = MaterialFunctionCall("Value (0-1)"=Multiply(A=Distance(A=@n12, B=@n13)) {constB=2}) {materialFunction=&MF_UI_EaseCurves}
  n15 = VectorParameter() {group=0 - Coloring, parameterName=Color 2, defaultValue=<rgba 0.6984 0.0779 1 1>}
  NamedRerouteDeclaration(None=Multiply(
    A=LinearInterpolate(A=@n10.RGB, B=@n15.RGB, Alpha=@n14.EaseInOutCubic)
    B=ScalarParameter() {defaultValue=1, parameterName=UnderGlow_Opacity}
   )) {name=Rift Gradient, nodeColor=<rgba 0 0 0 1>}
  n16 = Constant()
  n17 = Multiply(A=Divide(A=ScalarParameter() {defaultValue=0.4, parameterName=Modal Size})) {constB=-1}
  n18 = Add(
   A=MaterialFunctionCall() {materialFunction=&GetUserInterfaceUV}.Normalized UV
   B=LinearInterpolate(B=AppendVector(A=@n16, B=@n17), Alpha=NamedRerouteUsage()) {constB=0.5}
  )
  n19 = Distance(A=ComponentMask(None=@n18) {r=true}, B=Constant() {r=0.5})
  n20 = Multiply(A=Power(Base=Multiply(A=@n19) {constB=2}), B=ScalarParameter() {defaultValue=0.65, parameterName=Edge Intensity}) {constB=0.65}
  n21 = OneMinus(None=@n20)
  n22 = Multiply(A=Distance(A=ComponentMask(None=@n18) {g=true}, B=Constant() {r=0.5}), B=ScalarParameter() {defaultValue=4, parameterName=Line Thickness}) {constB=5}
  n23 = Power(Base=OneMinus(None=Multiply(A=@n21, B=@n22))) {constExponent=8}
  n24 = OneMinus(None=ScalarParameter() {defaultValue=1, parameterName=Line Reveal})
  n25 = SmoothStep(Min=LinearInterpolate(Alpha=@n24) {constA=-1}, Max=@n24, Value=NamedRerouteUsage())
  n26 = NamedRerouteDeclaration(None=SmoothStep(Min=NamedRerouteUsage(), Value=Multiply(A=@n23, B=@n25))) {name=Line, nodeColor=<rgba 0 0 0 1>}
  n27 = NamedRerouteUsage()
  NamedRerouteDeclaration(None=MaterialFunctionCall!(Blend=Min(A=@n23) {constB=0.85}, Base=Multiply(A=@n26, B=@n27)) {materialFunction=/Engine/Functions/Engine_MaterialFunctions03/Blends/Blend_ColorDodge}.Result) {name=Line Colors, nodeColor=<rgba 0 0 0 1>}
  NamedRerouteDeclaration(None=@n20) {name=Line Fades, nodeColor=<rgba 0 0 0 1>}
  n28 = Constant() {r=1}
  n29 = LinearInterpolate(
   B=ScalarParameter() {defaultValue=0.4, parameterName=Modal Size}
   Alpha=NamedRerouteDeclaration(None=ScalarParameter() {defaultValue=1, group=9 - Reveal Mask, parameterName=Modal Reveal Amount}) {name=Reveal Amount, nodeColor=<rgba 0 0 0 1>}
  ) {constA=0.01, constB=0.5}
  NamedRerouteDeclaration(None=MaterialFunctionCall(
    UVs=MaterialFunctionCall() {materialFunction=&GetUserInterfaceUV}.Normalized UV
    Texture Scale=AppendVector(A=@n28, B=@n29)
   ) {materialFunction=&ScaleUVsByCenter}.UVs) {name=ModalUV, nodeColor=<rgba 0 0 0 1>}
  n30 = NamedRerouteUsage()
  n31 = OneMinus(None=NamedRerouteUsage())
  n32 = Min(A=Step(Y=ComponentMask(None=@n30) {g=true}), B=Step(Y=ComponentMask(None=@n31) {g=true}))
  n33 = ComponentMask(None=NamedRerouteUsage()) {g=true}
  n34 = SmoothStep(Min=OneMinus(None=NamedRerouteUsage()), Value=@n33)
  n35 = LinearInterpolate(
   A=LinearInterpolate(B=ScalarParameter() {defaultValue=0.5, parameterName=Modal Top Fade}, Alpha=NamedRerouteUsage()) {constB=0.7}
   Alpha=@n34
  )
  n36 = Power(Base=SmoothStep(Value=@n33) {constMax=1.2, constMin=1}) {constExponent=0.2}
  n37 = NamedRerouteUsage()
  n38 = Multiply(A=Distance(A=ComponentMask(None=@n37) {r=true}, B=Constant() {r=0.5})) {constB=2}
  n39 = Power(Base=Multiply(A=OneMinus(None=@n36), B=Power(Base=@n38) {constExponent=1}))
  NamedRerouteDeclaration(None=LinearInterpolate(A=Multiply(A=@n32, B=@n35), Alpha=Multiply(A=@n34, B=@n39))) {name=Modal Mask, nodeColor=<rgba 0 0 0 1>}
  n40 = ScalarParameter() {defaultValue=0.9, parameterName=Underglow_Size}
  n41 = NamedRerouteUsage()
  n42 = MaterialFunctionCall(UV Channel=MaterialFunctionCall!(UV=@n41, Mirror Y, Swizzle UV) {materialFunction=/FortUILibrary/MaterialFunctions/Basics/MF_UI_Symmetry_2x1_New}.Symmetry) {materialFunction=&LinearGradient}
  n43 = OneMinus(None=@n42.VGradient)
  n44 = OneMinus(None=Multiply(A=Power(Base=@n42.UGradient) {constExponent=0.5}, B=Power(Base=@n43) {constExponent=0.5}))
  n45 = Min(A=Multiply(A=SmoothStep(Max=@n40, Value=@n44), B=Power(Base=@n42.VGradient) {constExponent=0.2})) {constB=0.999}
  n46 = ScalarParameter() {defaultValue=1, parameterName=Underlight_Opacity}
  n47 = Constant2Vector() {g=1, r=4}
  n48 = OneMinus(None=Distance(
    A=MaterialFunctionCall(UVs=@n41, Texture Scale=@n47) {materialFunction=&ScaleUVsByCenter}.UVs
    B=Constant() {r=1}
   ))
  n49 = LinearInterpolate(A=NamedRerouteUsage(), B=NamedRerouteUsage(), Alpha=Multiply(A=Power(Base=@n48) {constExponent=3}) {constB=0.1})
  n50 = NamedRerouteUsage()
  n51 = NamedRerouteUsage()
  n52 = OneMinus(None=Distance(
    A=MaterialFunctionCall(UV=@n51) {materialFunction=&MF_UI_UVScaling}.Scaled to Fit
    B=Constant() {r=0.5}
   ))
  n53 = Multiply(
   A=MaterialFunctionCall("Value (0-1)"=Saturate(None=@n52)) {materialFunction=&MF_UI_EaseCurves}.EaseInQuint
   B=ScalarParameter() {parameterName=CenterGlow_Opacity}
  ) {constB=0}
  NamedRerouteDeclaration(None=MaterialFunctionCall(Blend=Multiply(A=@n45, B=@n46), Base=LinearInterpolate(A=@n49, B=@n50, Alpha=@n53)) {materialFunction=&Blend_ColorDodge}.Result) {name=Modal Color, nodeColor=<rgba 0 0 0 1>}
  n54 = MaterialFunctionCall(
   UVs=MaterialFunctionCall() {materialFunction=&GetUserInterfaceUV}.Normalized UV
   Texture Scale=Constant2Vector() {g=2, r=1}
  ) {materialFunction=&ScaleUVsByCenter}
  n55 = AppendVector(A=OneMinus(None=ComponentMask(None=@n54.UVs) {r=true}), B=ComponentMask(None=@n54.UVs) {g=true})
  n56 = Constant()
  n57 = Power(Base=OneMinus(None=Distance(A=@n55, B=@n56))) {constExponent=4}
  n58 = AppendVector(A=OneMinus(None=ComponentMask(None=@n54.UVs) {r=true}), B=ComponentMask(None=@n54.UVs) {g=true})
  n59 = Constant() {r=1}
  n60 = Power(Base=OneMinus(None=Distance(A=@n58, B=@n59))) {constExponent=3}
  NamedRerouteDeclaration(None=Multiply(A=LinearInterpolate(A=@n57, Alpha=@n60), B=ScalarParameter() {defaultValue=0.35, parameterName=Blobs Opacity}) {constB=0.3}) {name=Blobs, nodeColor=<rgba 0 0 0 1>}
  NamedRerouteUsage()
  ScreenPosition()
 ]
}
